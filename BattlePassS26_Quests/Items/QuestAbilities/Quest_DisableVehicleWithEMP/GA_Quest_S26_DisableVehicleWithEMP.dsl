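; Season 26 Battle Pass quest ability that reports when the player disables a land vehicle with an EMP.
; On activation it waits on the activating pawn for a gameplay event and checks the payload target for the land terrain vehicle tag.
; A matching Fort vehicle triggers a DisableVehicleWithEMP custom verb broadcast carrying the player's tags and absolute forward speed, then a gameplay event send.
; Debug CAR / NOT CAR screen prints remain in the graph.
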(event Ability|EventActivateAbility
  (|CallParentFunction)
  (bind (_asynctaskproxy _payload) (Ability|Async|WaitGameplayEventToActor (Ability|GetActivatingPawn) (Variables|Default|GetEventTag))
    (:EventReceived
      (bind _bdidactorhavetags (Class|AthenaQuestFunctionLibrary|AthenaCheckActorHasTags (.Target (Utilities|Struct|BreakGameplayEventData _payload)) "(GameplayTags=((TagName=\"Vehicle.TerrainType.Land\")))"))
      (if _bdidactorhavetags
        (Development|PrintString "CAR" false false "(R=1.000000,G=0.793919,B=0.000000,A=1.000000)" 5.0)
        (bind _asfort_athena_vehicle (Utilities|Casting|CastToFortAthenaVehicle (.Target (Utilities|Struct|BreakGameplayEventData _payload)))
          (:then
            (bind _result (CustomVerb|AddTagsforSubject (CustomVerb|MakeCustomVerbMessage (Fortnite|GetFortPlayerControllerfromActor (Variables|GAVars|Data|GetOwningPlayerPawn)) "(TagName=\"CustomVerb.DisableVehicleWithEMP\")") "DisabledVehicle" (GameplayTags|GetOwnedGameplayTags (Variables|GAVars|Data|GetOwningPlayerPawn) :!class /Script/GameplayTags.GameplayTagAssetInterface)))
            (CustomVerb|AddValue _result "Speed" (Math|Float|Absolute(Float) (Vehicle|GetForwardSpeedKmh _asfort_athena_vehicle)))
            (CustomVerb|BroadcastCustomVerbMessage _result)
            (Ability|SendGameplayEventtoActor 0 "(TagName=\"CustomVerb.DisableVehicleWithEMP\")"))
          (:CastFailed
            _)))
      (Development|PrintString "NOT CAR" false false "(R=1.000000,G=0.396272,B=0.000000,A=1.000000)" 5.0))))

(event Ability|EventOnEndAbility
  (params (bWasCancelled :type Boolean)))
